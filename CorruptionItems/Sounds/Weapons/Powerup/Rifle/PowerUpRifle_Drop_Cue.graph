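# Sound cue for dropping the PowerUp Rifle weapon, classed as loot-collect audio and attenuated with shared weapon-drop settings.
# A mixer layers a pitched-down drop sound at prominent volume with a much quieter accent at normal pitch.
# The accent picks evenly at random between two "KevinStop" waves.
# The cue plays at reduced overall volume and references the shared Weapon_Drop_StopOldest concurrency set.
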SoundCue /CorruptionItems/Sounds/Weapons/Powerup/Rifle/PowerUpRifle_Drop_Cue
root=Mixer_0
nodes [
 {id=Mixer_0, children [Modulator_0,Modulator_2], properties {inputVolume [1.2,0.2]}}
 {id=Modulator_0, children [WavePlayer_3], properties {pitchMax=0.8, pitchMin=0.8}}
 {id=Modulator_2, children [Random_0], properties {pitchMax=1, pitchMin=1}}
 {id=Random_0, children [WavePlayer_1, WavePlayer_11], properties {weights [1,1]}}
 {id=WavePlayer_1, properties {soundWaveAssetPtr=&/PowerUpRifle_KevinStop_01}}
 {id=WavePlayer_11, properties {soundWaveAssetPtr=&/PowerUpRifle_KevinStop_03}}
 {id=WavePlayer_3, properties {soundWaveAssetPtr=&/PowerupRifle_Drop_01}}
]
properties {
 attenuationSettings=&Fort_Weapon_Drop_Att
 maxDistance=3000
 soundClassObject=&Fort_Loot_Collect
 volumeMultiplier=0.5
 concurrencyOverrides {resolutionRule=PreventNew}
 concurrencySet [&Weapon_Drop_StopOldest]
}
