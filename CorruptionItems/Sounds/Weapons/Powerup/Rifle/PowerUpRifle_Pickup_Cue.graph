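# Sound cue for picking up a powerup rifle, from the CorruptionItems weapon sounds.
# A single wave player feeds a Modulator that slightly randomizes pitch upward, on top of the cue's raised pitch and reduced volume.
# It uses the shared weapon pickup attenuation, the loot collect sound class and the PickupMaximum_Oldest concurrency set.
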SoundCue /CorruptionItems/Sounds/Weapons/Powerup/Rifle/PowerUpRifle_Pickup_Cue
root=Modulator_0
nodes [
 {id=Modulator_0, children [WavePlayer_1], properties {pitchMax=1.1, pitchMin=1}}
 {id=WavePlayer_1, properties {soundWaveAssetPtr=&/PowerupRifle_Pickup_02}}
]
properties {
 attenuationSettings=&Fort_Weapon_Pickup_Att
 maxDistance=3500
 pitchMultiplier=1.1
 soundClassObject=&Fort_Loot_Collect
 volumeMultiplier=0.5538
 concurrencyOverrides {maxCount=1, resolutionRule=PreventNew}
 concurrencySet [&PickupMaximum_Oldest]
}
