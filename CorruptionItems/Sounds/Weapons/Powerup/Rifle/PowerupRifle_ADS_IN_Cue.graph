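# Sound cue for the aim-down-sights-in foley of the Powerup Rifle, built from LMG power-up ADS-in waves.
# A Mixer combines a base wave, muted by a zero input volume, with an evenly weighted random pick of two layer waves, so only the layer is audible.
# A root Modulator varies the result, which plays pitched up at reduced volume under the weapon foley class with weapon reload attenuation.
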SoundCue /CorruptionItems/Sounds/Weapons/Powerup/Rifle/PowerupRifle_ADS_IN_Cue
root=Modulator_1
nodes [
 {id=Mixer_0, children [WavePlayer_0,Random_1], properties {inputVolume [0,1]}}
 {id=Modulator_1, children [Mixer_0]}
 {id=Random_1, children [WavePlayer_4,WavePlayer_5], properties {weights [1,1]}}
 {id=WavePlayer_0, properties {soundWaveAssetPtr=/CorruptionItems/Sounds/Weapons/LMG/PowerUpLMG/LMG_PowerUp_ADS_In_02}}
 {id=WavePlayer_4, properties {soundWaveAssetPtr=/CorruptionItems/Sounds/Weapons/LMG/PowerUpLMG/LMG_PowerUp_ADS_In_Layer_01}}
 {id=WavePlayer_5, properties {soundWaveAssetPtr=/CorruptionItems/Sounds/Weapons/LMG/PowerUpLMG/LMG_PowerUp_ADS_In_Layer_02}}
]
properties {
 attenuationSettings=&Fort_Weapon_Reload_Att
 maxDistance=2250
 pitchMultiplier=1.2
 soundClassObject=&Fort_WeaponFoley
 volumeMultiplier=0.376
}
