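# Sound cue for the Powerup Rifle's aim-down-sights exit, routed as weapon foley with weapon reload attenuation.
# It reuses LMG power-up waves, mixing the main ADS-out sound with a second layer that starts after a short fixed delay.
# A modulator adjusts the mixed result, while cue-level multipliers lower the overall pitch and volume.
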SoundCue /CorruptionItems/Sounds/Weapons/Powerup/Rifle/PowerupRifle_ADS_OUT_Cue
root=Modulator_1
nodes [
 {id=Delay_0, children [WavePlayer_2], properties {delayMax=0.1, delayMin=0.1}}
 {id=Mixer_0, children [WavePlayer_1,Delay_0], properties {inputVolume [1,1]}}
 {id=Modulator_1, children [Mixer_0], properties {pitchMin=1.05}}
 {id=WavePlayer_1, properties {soundWaveAssetPtr=/CorruptionItems/Sounds/Weapons/LMG/PowerUpLMG/LMG_PowerUp_ADS_Out}}
 {id=WavePlayer_2, properties {soundWaveAssetPtr=/CorruptionItems/Sounds/Weapons/LMG/PowerUpLMG/LMG_PowerUp_ADS_Out_Layer}}
]
properties {
 attenuationSettings=&Fort_Weapon_Reload_Att
 maxDistance=2250
 pitchMultiplier=0.8
 soundClassObject=&Fort_WeaponFoley
 volumeMultiplier=0.376
}
