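# Sound cue for the Powerup Rifle's FirePower gunshot, layering close, medium and distant recordings by distance.
# The close layer picks a random wave through close projectile attenuation and a third-person fire sound class.
# Medium and distant layers pick random waves, boost volume through modulators (medium also varies pitch), use their own attenuations, and share a distant-fire sound class.
# The cue raises overall pitch, uses local and global gunfire concurrency sets, and has debug enabled.
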SoundCue /CorruptionItems/Sounds/Weapons/Powerup/Rifle/PowerupRifle_FirePower_Cue
root=Mixer_1
nodes [
 {id=Attenuation_118, children [Random_0], properties {attenuationSettings=&Fort_Weapon_Projectile_Close_Athena_Att}}
 {
  id=Attenuation_119
  children [Modulator_0]
  properties {attenuationSettings=&Fort_Weapon_Projectile_Medium_Mono_Athena_Att}
 }
 {
  id=Attenuation_120
  children [Modulator_1]
  properties {attenuationSettings=&Fort_Weapon_Projectile_Distant_Mono_Athena_Att}
 }
 {id=Mixer_1, children [SoundClass_77, SoundClass_76], properties {inputVolume [1,0.7]}}
 {id=Mixer_4, children [Attenuation_119, Attenuation_120], properties {inputVolume [0.8,0.5]}}
 {id=Modulator_0, children [Random_5], properties {pitchMax=1.1, pitchMin=1, volumeMax=1.7, volumeMin=1.7}}
 {id=Modulator_1, children [Random_6], properties {volumeMax=1.8, volumeMin=1.8}}
 {id=Random_0, children [WavePlayer_1,WavePlayer_2,WavePlayer_3], properties {weights [1,1,1]}}
 {id=Random_5, children [WavePlayer_23, WavePlayer_24, WavePlayer_25, WavePlayer_26], properties {weights [1,1,1,1]}}
 {id=Random_6, children [WavePlayer_28, WavePlayer_29, WavePlayer_30], properties {weights [1,1,1]}}
 {id=SoundClass_76, children [Mixer_4], properties {soundClassOverride=&Fort_WeaponRanged_Fire_Distant}}
 {
  id=SoundClass_77
  children [Attenuation_118]
  properties {soundClassOverride=&Fort_WeaponRanged_Fire_3P_PSM}
 }
 {id=WavePlayer_1, properties {soundWaveAssetPtr=&/PowerupRifle_FirePower_02}}
 {id=WavePlayer_2, properties {soundWaveAssetPtr=&/PowerupRifle_FirePower_03}}
 {id=WavePlayer_23, properties {soundWaveAssetPtr=&/PowerupRifle_Fire_Medium02}}
 {id=WavePlayer_24, properties {soundWaveAssetPtr=&/PowerupRifle_Fire_Medium03}}
 {id=WavePlayer_25, properties {soundWaveAssetPtr=&/PowerupRifle_Fire_Medium04}}
 {id=WavePlayer_26, properties {soundWaveAssetPtr=&/PowerupRifle_Fire_Medium05}}
 {id=WavePlayer_28, properties {soundWaveAssetPtr=&/PowerupRifle_Fire_Distant01}}
 {id=WavePlayer_29, properties {soundWaveAssetPtr=&/PowerupRifle_Fire_Distant02}}
 {id=WavePlayer_3, properties {soundWaveAssetPtr=&/PowerupRifle_FirePower_04}}
 {id=WavePlayer_30, properties {soundWaveAssetPtr=&/PowerupRifle_Fire_Distant03}}
]
properties {
 bDebug=true
 maxDistance=27500
 pitchMultiplier=1.15
 volumeMultiplier=0.55
 concurrencyOverrides {bLimitToOwner=true, maxCount=3, resolutionRule=StopOldest}
 concurrencySet [&Weapon_Gunfire_Local_Conc, &Weapon_Gunfire_Global_Conc]
}
