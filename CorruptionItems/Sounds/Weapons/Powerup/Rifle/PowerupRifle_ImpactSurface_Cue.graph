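# Sound cue for surface impacts of the powerup rifle in CorruptionItems.
# After a short fixed delay, it mixes a random powerup impact wave, boosted in volume with slight pitch variation, with a slightly quieter, pitched-down random dirt impact borrowed from BlackMetal assault rifle sounds.
# It uses bullet impact attenuation, the impacts sound class and an owner-limited concurrency override that caps voices and scales down volume.
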SoundCue /CorruptionItems/Sounds/Weapons/Powerup/Rifle/PowerupRifle_ImpactSurface_Cue
root=Delay_3
nodes [
 {id=Delay_3, children [Mixer_0], properties {delayMax=0.1, delayMin=0.1}}
 {id=Mixer_0, children [Modulator_0,Modulator_1], properties {inputVolume [1,0.8]}}
 {id=Modulator_0, children [Random_1], properties {pitchMax=1.1, pitchMin=0.9, volumeMax=1.5, volumeMin=1.5}}
 {id=Modulator_1, children [Random_0], properties {pitchMax=0.8, pitchMin=0.7}}
 {id=Random_0, children [WavePlayer_0,WavePlayer_3,WavePlayer_5,WavePlayer_6], properties {weights [1,1,1,1]}}
 {id=Random_1, children [WavePlayer_1,WavePlayer_2,WavePlayer_4,WavePlayer_7,WavePlayer_8,WavePlayer_9], properties {weights [1,1,1,1,1,1]}}
 {
  id=WavePlayer_0
  properties {soundWaveAssetPtr=/Game/Sounds/Fort_Weapons/BlackMetal/Assault_Rifle/STW_BlackMetal_Ballistic_Proj_ImpactDirt_01}
 }
 {id=WavePlayer_1, properties {soundWaveAssetPtr=&/PowerupRifle_Impact_01}}
 {id=WavePlayer_2, properties {soundWaveAssetPtr=&/PowerupRifle_Impact_02}}
 {
  id=WavePlayer_3
  properties {soundWaveAssetPtr=/Game/Sounds/Fort_Weapons/BlackMetal/Assault_Rifle/STW_BlackMetal_Ballistic_Proj_ImpactDirt_03}
 }
 {id=WavePlayer_4, properties {soundWaveAssetPtr=&/PowerupRifle_Impact_03}}
 {
  id=WavePlayer_5
  properties {soundWaveAssetPtr=/Game/Sounds/Fort_Weapons/BlackMetal/Assault_Rifle/STW_BlackMetal_Ballistic_Proj_ImpactDirt_04}
 }
 {
  id=WavePlayer_6
  properties {soundWaveAssetPtr=/Game/Sounds/Fort_Weapons/BlackMetal/Assault_Rifle/STW_BlackMetal_Ballistic_Proj_ImpactDirt_02}
 }
 {id=WavePlayer_7, properties {soundWaveAssetPtr=&/PowerupRifle_Impact_04}}
 {id=WavePlayer_8, properties {soundWaveAssetPtr=&/PowerupRifle_Impact_05}}
 {id=WavePlayer_9, properties {soundWaveAssetPtr=&/PowerupRifle_Impact_06}}
]
properties {
 attenuationSettings=/Game/Sounds/Attenuation/Fort_Bullet_Impact_Athena_Att
 bOverrideConcurrency=true
 maxDistance=4500
 soundClassObject=&Fort_Impacts
 volumeMultiplier=0.9
 concurrencyOverrides {bLimitToOwner=true, maxCount=4, volumeScale=0.7, volumeScaleAttackTime=0.2}
 concurrencySet [null]
}
