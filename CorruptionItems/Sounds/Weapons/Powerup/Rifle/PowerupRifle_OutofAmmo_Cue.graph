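# Sound cue for the out-of-ammo sound of the Powerup Rifle.
# A mixer layers one of four random sniper out-of-ammo waves with a quieter, pitched-up LMG power-up out-of-ammo wave, then feeds a root modulator that keeps pitch unchanged.
# The cue uses a third-person ranged weapon fire sound class and overrides concurrency to allow two instances, stopping the oldest.
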SoundCue /CorruptionItems/Sounds/Weapons/Powerup/Rifle/PowerupRifle_OutofAmmo_Cue
root=Modulator_1
nodes [
 {id=Mixer_0, children [Random_0,Modulator_0], properties {inputVolume [1,0.8]}}
 {id=Modulator_0, children [WavePlayer_4], properties {pitchMax=1.2, pitchMin=1.2, volumeMin=1}}
 {id=Modulator_1, children [Mixer_0], properties {pitchMax=1, pitchMin=1}}
 {id=Random_0, children [WavePlayer_0,WavePlayer_1,WavePlayer_2,WavePlayer_3], properties {weights [1,1,1,1]}}
 {id=WavePlayer_0, properties {soundWaveAssetPtr=/Game/Sounds/Fort_Weapons/OutofAmmo/Sniper_OutofAmmo_01}}
 {id=WavePlayer_1, properties {soundWaveAssetPtr=/Game/Sounds/Fort_Weapons/OutofAmmo/Sniper_OutofAmmo_02}}
 {id=WavePlayer_2, properties {soundWaveAssetPtr=/Game/Sounds/Fort_Weapons/OutofAmmo/Sniper_OutofAmmo_03}}
 {id=WavePlayer_3, properties {soundWaveAssetPtr=/Game/Sounds/Fort_Weapons/OutofAmmo/Sniper_OutofAmmo_04}}
 {id=WavePlayer_4, properties {soundWaveAssetPtr=/CorruptionItems/Sounds/Weapons/LMG/PowerUpLMG/LMG_PowerUp_OutofAmmo_01}}
]
properties {
 bOverrideConcurrency=true
 maxDistance=2097152
 soundClassObject=&Fort_WeaponRanged_Fire_3P_PSM
 volumeMultiplier=0.6
 attenuationOverrides {falloffDistance=1600}
 concurrencyOverrides {maxCount=2, resolutionRule=StopOldest}
}
