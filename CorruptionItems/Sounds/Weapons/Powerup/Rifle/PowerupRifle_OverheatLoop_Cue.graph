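# Sound cue for the looping overheat sound of the Powerup Rifle, using a first-person special weapon sound class.
# A looping powerup wave feeds a continuous modulator, which maps the Overheat parameter to both pitch and volume.
# The cue lowers base pitch, boosts volume and uses close weapon projectile attenuation.
# Its concurrency override prevents new instances once the limit is reached.
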SoundCue /CorruptionItems/Sounds/Weapons/Powerup/Rifle/PowerupRifle_OverheatLoop_Cue
root=ModulatorContinuous_2
nodes [
 {
  id=ModulatorContinuous_2
  children [WavePlayer_2]
  properties {
   pitchModulationParams {default=0, maxOutput=1.1, minInput=0.25, minOutput=0.85, parameterName=Overheat}
   volumeModulationParams {default=0.4, minInput=0.25, parameterName=Overheat}
  }
 }
 {id=WavePlayer_2, properties {bLooping=true, soundWaveAssetPtr=&/PowerupRifle_PowerupLoop_01}}
]
properties {
 attenuationSettings=&Fort_Weapon_Projectile_Close_Athena_Att
 bOverrideConcurrency=true
 maxDistance=4950
 pitchMultiplier=0.7
 soundClassObject=/Game/Sounds/FortSoundClassesAndModes/Fort_LTM_BR_InGame_Special_Weapon_1P
 volumeMultiplier=2
 attenuationOverrides {distanceAlgorithm=NaturalSound, falloffDistance=4000, attenuationShapeExtents {x=1000}}
 concurrencyOverrides {resolutionRule=PreventNew}
}
