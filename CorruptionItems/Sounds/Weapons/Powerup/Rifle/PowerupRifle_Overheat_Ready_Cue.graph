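# Sound cue for the corruption powerup rifle, named for the ready moment tied to its overheat.
# A mixer layers two powerup-start waves at equal volume, with a modulator randomizing only the main wave to a slightly raised pitch.
# It uses the weapon foley class and reload attenuation, with cue-level attenuation overrides.
# Owner-limited concurrency allows one instance at a time.
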SoundCue /CorruptionItems/Sounds/Weapons/Powerup/Rifle/PowerupRifle_Overheat_Ready_Cue
root=Mixer_0
nodes [
 {id=Mixer_0, children [Modulator_0,WavePlayer_1], properties {inputVolume [1,1]}}
 {id=Modulator_0, children [WavePlayer_0], properties {pitchMax=1.2, pitchMin=1.1}}
 {id=WavePlayer_0, properties {soundWaveAssetPtr=&/PowerupRifle_PowerupStart_03}}
 {id=WavePlayer_1, properties {soundWaveAssetPtr=&/PowerupRifle_PowerupStartLayer_05}}
]
properties {
 attenuationSettings=&Fort_Weapon_Reload_Att
 bOverrideAttenuation=true
 bOverrideConcurrency=true
 maxDistance=2100
 soundClassObject=&Fort_WeaponFoley
 volumeMultiplier=0.501
 attenuationOverrides {bEnableReverbSend=false, falloffDistance=1500, stereoSpread=600, attenuationShapeExtents {x=600}}
 concurrencyOverrides {bLimitToOwner=true, maxCount=1, resolutionRule=StopLowestPriorityThenPreventNew}
}
