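# Sound cue for the Powerup Rifle's overheat release, layering a close and a medium-distance version of the sound.
# The close branch mixes two overheat waves through close projectile attenuation and the third-person ranged-fire sound class.
# A quieter medium-range wave uses its own mono attenuation and the distant ranged-fire class.
# Overridden concurrency caps simultaneous plays per owner, stopping the oldest, and the cue's debug flag is on.
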SoundCue /CorruptionItems/Sounds/Weapons/Powerup/Rifle/PowerupRifle_Overheat_Release_Cue
root=Mixer_1
nodes [
 {id=Attenuation_118, children [Mixer_0], properties {attenuationSettings=&Fort_Weapon_Projectile_Close_Athena_Att}}
 {
  id=Attenuation_119
  children [WavePlayer_2]
  properties {attenuationSettings=&Fort_Weapon_Projectile_Medium_Mono_Athena_Att}
 }
 {id=Mixer_0, children [WavePlayer_1,WavePlayer_8], properties {inputVolume [1,1.6]}}
 {id=Mixer_1, children [SoundClass_77, SoundClass_76], properties {inputVolume [1,0.5]}}
 {
  id=SoundClass_76
  children [Attenuation_119]
  properties {soundClassOverride=&Fort_WeaponRanged_Fire_Distant}
 }
 {
  id=SoundClass_77
  children [Attenuation_118]
  properties {soundClassOverride=&Fort_WeaponRanged_Fire_3P_PSM}
 }
 {id=WavePlayer_1, properties {soundWaveAssetPtr=&/PowerupRifle_Overheat_02}}
 {id=WavePlayer_2, properties {soundWaveAssetPtr=&/PowerupRifle_Overheat_Medium_02}}
 {id=WavePlayer_8, properties {soundWaveAssetPtr=&/PowerupRifle_OverheatLayer_06}}
]
properties {
 bDebug=true
 bOverrideConcurrency=true
 maxDistance=10579.2871
 volumeMultiplier=0.15
 attenuationOverrides {bEnableReverbSend=false, distanceAlgorithm=Logarithmic, falloffDistance=1500, stereoSpread=600, attenuationShapeExtents {x=600}}
 concurrencyOverrides {bLimitToOwner=true, maxCount=2, resolutionRule=StopOldest}
}
