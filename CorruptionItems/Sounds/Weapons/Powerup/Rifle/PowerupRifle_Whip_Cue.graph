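# Sound cue for the Powerup Rifle's projectile whip sound, grouped under bullet whiz-by concurrency and the ranged-weapon projectile effects sound class.
# A quality-level node routes the first two levels to six random Heavy whip waves and the last two to two regular ones.
# Each random pool feeds its own pitch modulator, with the Heavy pool pitched higher.
# Overridden attenuation turns off volume falloff with distance.
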SoundCue /CorruptionItems/Sounds/Weapons/Powerup/Rifle/PowerupRifle_Whip_Cue
root=QualityLevel_0
nodes [
 {id=Modulator_0, children [Random_0], properties {pitchMax=1.6, pitchMin=1.2}}
 {id=Modulator_1, children [Random_1], properties {pitchMax=1.2, pitchMin=1}}
 {id=QualityLevel_0, children [Modulator_0,Modulator_0,Modulator_1,Modulator_1]}
 {id=Random_0, children [WavePlayer_0,WavePlayer_1,WavePlayer_2,WavePlayer_3,WavePlayer_4,WavePlayer_6], properties {weights [1,1,1,1,1,1]}}
 {id=Random_1, children [WavePlayer_11, WavePlayer_12], properties {bShouldExcludeFromBranchCulling=true, weights [1,1]}}
 {id=WavePlayer_0, properties {soundWaveAssetPtr=&/PowerupRifle_Whip_Heavy_01}}
 {id=WavePlayer_1, properties {soundWaveAssetPtr=&/PowerupRifle_Whip_Heavy_02}}
 {id=WavePlayer_11, properties {soundWaveAssetPtr=&/PowerupRifle_Whip_01}}
 {id=WavePlayer_12, properties {soundWaveAssetPtr=&/PowerupRifle_Whip_02}}
 {id=WavePlayer_2, properties {soundWaveAssetPtr=&/PowerupRifle_Whip_Heavy_03}}
 {id=WavePlayer_3, properties {soundWaveAssetPtr=&/PowerupRifle_Whip_Heavy_04}}
 {id=WavePlayer_4, properties {soundWaveAssetPtr=&/PowerupRifle_Whip_Heavy_05}}
 {id=WavePlayer_6, properties {soundWaveAssetPtr=&/PowerupRifle_Whip_Heavy_06}}
]
properties {
 bOverrideAttenuation=true
 maxDistance=2097152
 soundClassObject=&Fort_WeaponRanged_Projectile_Effects
 volumeMultiplier=0.9
 attenuationOverrides {bAttenuate=false, stereoSpread=0}
 concurrencySet [/Game/Sounds/Concurrency_Rules/Athena_BulletWhizBy_Conc]
}
